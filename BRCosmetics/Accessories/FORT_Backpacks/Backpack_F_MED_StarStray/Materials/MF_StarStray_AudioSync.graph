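# Material function for the StarStray backpack that replaces the incoming emissive with an animated, audio-reactive equalizer effect.
# Pre-skinned local position drives procedural masks and level fills, combined with an FX mask texture and a normal-distorted nebula background.
# A sine-switched looping timer builds the idle animation.
# Music spectrum and event amplitude from parameter collections supply the emote spectrum and threshold, with smooth and 13-step stepped bands.
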MaterialFunction /BRCosmetics/Accessories/FORT_Backpacks/Backpack_F_MED_StarStray/Materials/MF_StarStray_AudioSync
graph {nodes [
  n1 = NamedRerouteUsage()
  n2 = NamedRerouteUsage()
  n3 = LinearInterpolate(A=Multiply(A=@n2) {constB=0.05}, B=Desaturation(None=@n2, Fraction=Constant() {r=-0.2}), Alpha=NamedRerouteUsage())
  n4 = NamedRerouteUsage()
  FunctionOutput(None=SetMaterialAttributes!(MaterialAttributes=NamedRerouteUsage(), EmissiveColor=LinearInterpolate(A=@n1, B=@n3, Alpha=@n4)) {attributeSetTypes [#1]})
  n5 = NamedRerouteUsage()
  n6 = Multiply(A=NamedRerouteUsage(), B=Power(Base=Multiply(A=@n5) {constB=2}))
  n7 = Add(A=NamedRerouteUsage(), B=Multiply(A=Constant3Vector() {constant=<rgba 0.2394 0.0234 0.25 1>}, B=@n6))
  NamedRerouteDeclaration(None=LinearInterpolate(B=LinearInterpolate(B=@n7, Alpha=@n6), Alpha=NamedRerouteUsage())) {name=BGActivation, nodeColor=<rgba 0.9922 0 1 1>}
  NamedRerouteDeclaration(None=LinearInterpolate(A=NamedRerouteUsage(), B=NamedRerouteUsage(), Alpha=NamedRerouteUsage())) {name=Spectrum FX, nodeColor=<rgba 0.4062 1 0 1>}
  n8 = GetMaterialAttributes(None=FunctionInput() {inputName=inMA, inputType=FunctionInput_MaterialAttributes}) {attributeGetTypes [#1]}
  NamedRerouteDeclaration(None=@n8.EmissiveColor) {name=inEmissive, nodeColor=<rgba 0 1 0.25 1>}
  NamedRerouteDeclaration(None=@n8.MaterialAttributes) {name=InMA, nodeColor=<rgba 0.9219 1 0 1>}
  n9 = NamedRerouteDeclaration(None=VertexInterpolator(VS=PreSkinnedPosition()).PS) {name=UvLocalCenter, nodeColor=<rgba 1 0 0.3516 1>}
  NamedRerouteDeclaration(None=Divide(A=Add(A=@n9) {constB=7.1}) {constB=14}) {name=UVLocal, nodeColor=<rgba 1 0 0.0938 1>}
  n10 = NamedRerouteUsage()
  n11 = ComponentMask(None=Divide(A=@n10) {constB=12}) {r=true}
  n12 = ComponentMask(None=Divide(A=@n10) {constB=16}) {b=true}
  n13 = Min(A=Max(A=@n11, B=Multiply(A=@n11) {constB=-1}), B=Max(A=@n12, B=Multiply(A=@n12) {constB=-1}))
  n14 = AppendVector(A=Multiply(A=@n10) {constB=0.21}, B=Multiply(A=ComponentMask(None=@n10) {r=true}) {constB=0.25})
  n15 = Constant()
  n16 = Multiply(A=OneMinus(None=Distance(A=@n14, B=@n15))) {constB=-0.5}
  NamedRerouteDeclaration(None=Multiply(A=Add(A=@n13, B=@n16), B=NamedRerouteUsage())) {name=BGMask, nodeColor=<rgba 1 0 0.7031 1>}
  n17 = Multiply(A=Time(), B=ScalarParameter() {defaultValue=0.2, group=MF_StarStray_AudioSync, parameterName=IdleSpeed, sortPriority=0}) {constB=0.3}
  n18 = Frac(None=Multiply(A=@n17) {constB=-2})
  n19 = Step(X=Sine(None=@n17))
  NamedRerouteDeclaration(None=LinearInterpolate(B=Multiply(A=@n18) {constB=2}, Alpha=@n19) {constB=0}) {name=InvertedTime, nodeColor=<rgba 0.3359 0.01 1 1>}
  n20 = TextureSample() {
   samplerType=SAMPLERTYPE_Masks
   texture=/BRCosmetics/Accessories/FORT_Backpacks/Backpack_F_MED_StarStray/Textures/T_F_MED_StarStray_Backpack_FX
  }
  NamedRerouteDeclaration(None=@n20.R) {name=TextureFXRChannel, nodeColor=<rgba 0.984 0 0 1>}
  NamedRerouteDeclaration(None=@n20.B) {name=TextureFXBChannel, nodeColor=<rgba 0 0 1 1>}
  n21 = Multiply(A=Time()) {constB=0.01}
  n22 = Add() {constB=0.33}
  n23 = TextureCoordinate() {uTiling=3, vTiling=3}
  n24 = TextureSample(UVs=Add(A=AppendVector(A=@n21, B=@n22), B=@n23)) {samplerType=SAMPLERTYPE_Normal, texture=/Game/Characters/Player/Male/Medium/Bodies/M_MED_BiruFang/Textures/T_BiruFang_EmissiveEffect_N}
  n25 = Multiply(A=@n24.RGB) {constB=0.1}
  n26 = TextureSample(UVs=Add(A=ComponentMask(None=@n25) {g=true, r=true}, B=@n23)) {
   automaticViewMipBias=false
   texture=/BRCosmetics/Characters/Player/Female/Medium/Bodies/F_MED_StarStray/Skins/Galaxy/Textures/FN_Nebula_StarStrayGalaxy_Purple
  }
  NamedRerouteDeclaration(None=@n26.RGB) {name=BGFX, nodeColor=<rgba 0 1 0.7656 1>}
  n27 = Multiply(A=@n17) {constB=2}
  NamedRerouteDeclaration(None=LinearInterpolate(A=Frac(None=@n27), Alpha=@n19) {constB=0}) {name=Time, nodeColor=<rgba 0.3359 1 0 1>}
  n28 = ComponentMask(None=NamedRerouteUsage()) {b=true}
  n29 = NamedRerouteUsage()
  n30 = OneMinus(None=ComponentMask(None=NamedRerouteUsage()) {b=true})
  n31 = Add(A=@n30)
  n32 = NamedRerouteUsage()
  NamedRerouteDeclaration(None=Add(A=Step(Y=@n28, X=@n29), B=SmoothStep(Min=@n30, Max=@n31, Value=@n32))) {name=IdleAnim, nodeColor=<rgba 1 0 0.8672 1>}
  n33 = ComponentMask(None=NamedRerouteUsage()) {b=true}
  n34 = Power(Base=NamedRerouteUsage()) {constExponent=0.4}
  n35 = Multiply(A=SmoothStep(Min=@n33, Max=Add(A=@n33) {constB=0.5}, Value=NamedRerouteDeclaration(None=@n34) {name=HighestSpectrum, nodeColor=<rgba 0 0.8828 1 1>})) {constB=1.2}
  n36 = ComponentMask(None=NamedRerouteUsage()) {b=true}
  n37 = NamedRerouteUsage()
  n38 = Multiply(A=SmoothStep(Min=@n36, Max=Add(A=@n36) {constB=0.1}, Value=Power(Base=@n37) {constExponent=0.4})) {constB=1.2}
  n39 = NamedRerouteUsage()
  n40 = Multiply(A=NamedRerouteUsage()) {constB=13}
  n41 = Floor(None=Add(A=@n40))
  n42 = NamedRerouteUsage()
  n43 = Floor(None=@n40)
  n44 = Subtract(
   A=Step(Y=ComponentMask(None=@n39) {b=true}, X=Divide(A=@n41) {constB=13})
   B=Step(Y=ComponentMask(None=@n42) {b=true}, X=Divide(A=@n43) {constB=13})
  )
  NamedRerouteDeclaration(None=Add(A=Add(A=@n35, B=@n38), B=Saturate(None=@n44))) {name=EmoteAnim, nodeColor=<rgba 0 0.7489 1 1>}
  n45 = CollectionParameter() {collection=/Game/Athena/Events/AudioAnalysis/MPC_Audio_Reactive_Parameters, parameterName=Spectrum_Music}
  n46 = Multiply(A=ComponentMask(None=@n45) {r=true}, B=ComponentMask(None=@n45) {g=true})
  n47 = Multiply(A=ComponentMask(None=@n45) {b=true}, B=ComponentMask(None=@n45) {a=true})
  n48 = Max(A=@n46, B=@n47)
  n49 = Min(A=@n46, B=@n47)
  NamedRerouteDeclaration(None=Max(
    A=NamedRerouteDeclaration(None=@n48) {name=MaxSpectrum, nodeColor=<rgba 0 0.0391 1 1>}
    B=NamedRerouteDeclaration(None=@n49) {name=MinSpectrum, nodeColor=<rgba 0 1 0.1094 1>}
   )) {name=EmotePlayingSpectrum, nodeColor=<rgba 0 1 0.9063 1>}
  n50 = CollectionParameter() {collection=&FortniteMaterialParameters, parameterName=EventAudio_Amplitude}
  n51 = Multiply(A=@n50) {constB=2}
  n52 = ScalarParameter() {defaultValue=1, group=MF_StarStray_AudioSync, parameterName=EmoteThreshold}
  NamedRerouteDeclaration(None=Saturate(None=Power(Base=@n51, Exp=@n52))) {name=EqualizerEmoteThreshold, nodeColor=<rgba 0 0.7187 1 1>}
 ]}
